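; Front-end cosmetic display Blueprint for the Build-A-Brella umbrella glider that plays sound and effects when a variant changes.
; On each variant change, the new variant tag is matched against canopy, shaft, topper and handle tag sets to set the audio component's sound, then play it.
; The variant channel selects the matching Niagara transition effect, with bottom cap mapped to handle.
; All effects spawn at the skeletal mesh's top-cap socket location.
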(event EventBeginPlay)

(event Collision|EventActorBeginOverlap
  (params (OtherActor :type Actor%20Object%20Reference)))

(event EventTick
  (params (DeltaSeconds :type Float%20%28single-precision%29)))

(event MannequinNotifications|EventOnVariantChanged
  (params (VariantChannel :type Gameplay%20Tag%20Structure) (OldVariantTag :type Gameplay%20Tag%20Structure) (NewVariantTag :type Gameplay%20Tag%20Structure))
  (CallFunction|UpdateAudio NewVariantTag)
  (CallFunction|UpdateTransitionVFX VariantChannel))

(fn UpdateAudio
  (params (Variant%20Channel :type Gameplay%20Tag%20Structure))
  (Utilities|FlowControl|Sequence
    (:then_0
      (if (GameplayTags|MatchesAnyTags Variant%20Channel (Variables|Default|GetCanopyTags) true)
        (Audio|Components|Audio|SetSound (Variables|Default|GetPlayAudio) (Variables|Sound|GetCanopySound))))
    (:then_1
      (if (GameplayTags|MatchesAnyTags Variant%20Channel (Variables|Default|GetShaftTags) true)
        (Audio|Components|Audio|SetSound (Variables|Default|GetPlayAudio) (Variables|Sound|GetShaftSound))))
    (:then_2
      (if (GameplayTags|MatchesAnyTags Variant%20Channel (Variables|Default|GetTopperTags) true)
        (Audio|Components|Audio|SetSound (Variables|Default|GetPlayAudio) (Variables|Sound|GetTopperSound))))
    (:then_3
      (if (GameplayTags|MatchesAnyTags Variant%20Channel (Variables|Default|GetHandleTags) true)
        (Audio|Components|Audio|SetSound (Variables|Default|GetPlayAudio) (Variables|Sound|GetHandleSound)))))
  (Audio|Components|Audio|Play (Variables|Default|GetPlayAudio) 0.0))

(fn UpdateTransitionVFX
  (params (Variant%20Channel :type Gameplay%20Tag%20Structure))
  (bind _skeletalmeshcomponent (Variables|FortCosmeticDisplayActor|GetSkeletalMeshComponent))
  (bind _returnvalue (Transformation|GetSocketLocation _skeletalmeshcomponent "mesh_top_cap"))
  (Utilities|FlowControl|Sequence
    (:then_0
      (if (GameplayTags|MatchesAnyTags Variant%20Channel (Variables|Default|GetCanopyVariantName) true)
        (Niagara|SpawnSystematLocation "/BRCosmetics/Gadgets/Assets/VinderTech_GliderChute/Glider_Umbrella_Build_A_Brella/FX/NS_Frontend_Build_A_Brella_Transition_Canopy.NS_Frontend_Build_A_Brella_Transition_Canopy" _returnvalue)))
    (:then_1
      (if (GameplayTags|MatchesAnyTags Variant%20Channel (Variables|Default|GetShaftVariantName) true)
        (Niagara|SpawnSystematLocation "/BRCosmetics/Gadgets/Assets/VinderTech_GliderChute/Glider_Umbrella_Build_A_Brella/FX/NS_Frontend_Build_A_Brella_Transition_Shaft.NS_Frontend_Build_A_Brella_Transition_Shaft" _returnvalue)))
    (:then_2
      (if (GameplayTags|MatchesAnyTags Variant%20Channel (Variables|Default|GetTopperVariantName) true)
        (Niagara|SpawnSystematLocation "/BRCosmetics/Gadgets/Assets/VinderTech_GliderChute/Glider_Umbrella_Build_A_Brella/FX/NS_Frontend_Build_A_Brella_Transition_Topper.NS_Frontend_Build_A_Brella_Transition_Topper" _returnvalue)))
    (:then_3
      (if (GameplayTags|MatchesAnyTags Variant%20Channel (Variables|Default|GetBottomCapVariantName) true)
        (Niagara|SpawnSystematLocation "/BRCosmetics/Gadgets/Assets/VinderTech_GliderChute/Glider_Umbrella_Build_A_Brella/FX/NS_Frontend_Build_A_Brella_Transition_Handle.NS_Frontend_Build_A_Brella_Transition_Handle" _returnvalue)))))

(fn ConstructionScript)
