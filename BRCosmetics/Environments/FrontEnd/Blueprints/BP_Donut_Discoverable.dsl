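; BP_Donut_Discoverable is a front-end environment actor that responds to special interactions and reports quest progress.
; When interaction is allowed, it asynchronously pushes an optional details widget as a modal and sends a client stat event.
; Successful credit plays sound and effects, then optionally hides the actor or suppresses further interaction.
; It also calls an empty commit event, runs an optional cooldown, gates navigation requests, and forwards notifications to linked discoverables.
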(event EventBeginPlay
  (bind _output_get (Variables|Default|SetPlayerController (Utilities|Array|Get(acopy) (Fortnite|GetLocalFortPlayerControllers true))))
  (Variables|Default|SetQuestManager (Quests|GetQuestManager _output_get "Invalid")))

(event EventHandleObjectSpecialInteraction
  (params (ActionTag :type Gameplay%20Tag%20Structure) (CameraTag :type Gameplay%20Tag%20Structure))
  (bind _returnvalue (Fortnite|GetFirstLocalFortPlayerController true))
  (Variables|Default|SetInteracted_ActionTag ActionTag)
  (Variables|Default|SetInteracted_CameraTag CameraTag)
  (if (and (Variables|Default|GetIsInteractAllowed) (not (Variables|Default|GetInteractionSuppressed)))
    (Utilities|FlowControl|Sequence
      (:then_0
        (if (Utilities|IsValidSoftClassReference (Variables|Default|GetDetailsWidgetClass))
          (bind _userwidget (Class|FortAsyncActionCreateWidgetAsync|CreateWidgetAsync (Variables|Default|GetDetailsWidgetClass) _returnvalue)
            (:OnComplete
              (FortUIManagement|TryPushWidgetasModalPanelforPlayer _returnvalue _userwidget)))))
      (:then_1
        (Quests|SendClientStatEvent (Variables|Default|GetQuestManager)
          (:OnSuccess
            (Audio|PlaySound2D (Variables|Default|GetSoundToPlayOnQuestCredit))
            (Effects|Components|ParticleSystem|SpawnFXAtLocation (Variables|Default|GetSuccessEmitter) (Transformation|GetActorLocation))
            (Utilities|FlowControl|Sequence
              (:then_0
                (if (Variables|Default|GetHideAfterObjectiveComplete)
                  (Rendering|SetActorHiddenInGame true)
                  (Variables|Default|SetIsNavAllowed false)
                  (Variables|Default|SetIsInteractAllowed false)))
              (:then_1
                (if (Variables|Default|GetSuppressInteractAfterObjectiveComplete)
                  (Variables|Default|SetInteractionSuppressed true)))))))
      (:then_2
        (CallFunction|CommitonSpecialInteraction)
        (if (and (> (Variables|Default|GetInteractCooldown) 0.0) (not (Variables|Default|GetHideAfterObjectiveComplete)))
          (Variables|Default|SetIsInteractAllowed false)
          (Utilities|Time|SetTimerbyEvent (delegate Custom|ResetCooldown) (Variables|Default|GetInteractCooldown)))))))

(event Custom|Commit_OnSpecialInteraction)

(event Custom|ResetCooldown
  (Variables|Default|SetIsInteractAllowed true))

(event Custom|NotifiedByDiscoverable
  (params (ActionTag :type Gameplay%20Tag%20Structure) (CameraTag :type Gameplay%20Tag%20Structure) (bIsStart :type Boolean))
  (for _array_element (Variables|Default|GetOtherDiscoverablesToNotify)
    (CallFunction|NotifiedbyDiscoverable _array_element ActionTag CameraTag bIsStart)))

(fn CanCommitNavigationRequest
  (return (Variables|Default|GetIsNavAllowed)))

(fn ConstructionScript)
