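; BP_Frontend_Adventure_Dice is an interactive adventure dice object for the front end.
; A special interaction passes an interact-allowed gate, locks it, fires the CompletedRoll dispatcher and commits the interaction.
; A timer set to the reset delay then fires GameReset and unlocks interaction.
; Roll visuals and results are not handled in this graph; navigation requests are always committable.
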(event EventBeginPlay)

(event EventHandleObjectSpecialInteraction
  (params (ActionTag :type Gameplay%20Tag%20Structure) (CameraTag :type Gameplay%20Tag%20Structure))
  (if (Variables|Default|GetIsInteractAllowed)
    (Variables|Default|SetIsInteractAllowed false)
    (Utilities|FlowControl|Sequence
      (:then_0
        (Default|CallCompletedRoll)
        (Utilities|Time|SetTimerbyEvent (delegate Custom|ResetGame) (Variables|Default|GetResetDelay)))
      (:then_1
        (CallFunction|CommitonSpecialInteraction)))))

(event Custom|ResetGame
  (Default|CallGameReset)
  (Variables|Default|SetIsInteractAllowed true))

(event Custom|Commit_OnSpecialInteraction)

(fn CanCommitNavigationRequest
  (return true))

(fn ConstructionScript)
