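; Front-end navigation Blueprint for Battle Pass doors that show purchase state and react to the cursor.
; Construction creates dynamic materials and previews the state via a debug flag; begin play, player init and inventory updates recheck season book ownership.
; SetBattlePassVisualState switches chrome, light brightness, spotlight colors and visibility, hiding the glint when purchased.
; Hover and click drive door and glint timelines; the open and close sound events are empty here.
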(event Collision|EventActorBeginOverlap
  (params (OtherActor :type Actor%20Object%20Reference))
  (|CallParentFunction OtherActor) ; This node is disabled and will not be called.\nDrag off pins to build functionality.
)

(event EventTick
  (params (DeltaSeconds :type Float%20%28single-precision%29))
  (|CallParentFunction DeltaSeconds) ; This node is disabled and will not be called.\nDrag off pins to build functionality.
)

(event EventBeginPlay
  (|CallParentFunction)
  (Utilities|FlowControl|Sequence
    (:then_0
      ;[1 1 1] Check for battlepass purchased and execute visual updates
      (EventDispatchers|BindEventtoOnInventoryUpdated (delegate Custom|UpdateVisuals)))
    (:then_1
      (CallFunction|SetBattlePassVisualState (CallFunction|HasPurchasedSeasonBook)))))

(event Custom|UpdateVisuals
  ;[1 1 1] Check for battlepass purchased and execute visual updates
  (CallFunction|SetBattlePassVisualState (CallFunction|HasPurchasedSeasonBook)))

(event EventOnInitializeForPlayer
  ;[1 1 1] Check for battlepass purchased and execute visual updates
  (CallFunction|SetBattlePassVisualState (CallFunction|HasPurchasedSeasonBook)))

(event MouseInput|EventActorBeginCursorOver
  (CallFunction|OpenDoorSound)
  (Utilities|FlowControl|Sequence
    (:then_0
      (bind (_direction _newtrack_0) (Timeline|FadeGlint Play)
        (:Update
          (Rendering|Material|SetScalarParameterValueonMaterials (Variables|Default|GetDoor_Glint) "Opacity" _newtrack_0))))
    (:then_1
      (bind (_direction_2 _newtrack_0_2 _timeinxeconds) (Timeline|Timeline_0 Play)
        (:Update
          (Variables|Default|SetTimeinSeconds _timeinxeconds)
          (Rendering|Material|SetScalarParameterValue (Variables|Default|GetMID_1) "OpenMouseOver" _newtrack_0_2 :!class /Script/Engine.MaterialInstanceDynamic))))))

(event MouseInput|EventActorEndCursorOver
  (Utilities|FlowControl|Gate false
    (:Exit
      (CallFunction|CloseDoorSound)
      (Utilities|FlowControl|Sequence
        (:then_0
          (Timeline|FadeGlint Reverse))
        (:then_1
          (Timeline|Timeline_0 Reverse))))))

(event MouseInput|EventActorOnClicked
  (params (ButtonPressed :type Key%20Structure))
  (Utilities|FlowControl|Sequence
    (:then_0
      (Timeline|Timeline_0 Play))
    (:then_1
      _)
    (:then_2
      (Utilities|FlowControl|Delay 1.5)
      (Utilities|FlowControl|Sequence
        (:then_0
          _)
        (:then_1
          (Timeline|Timeline_0 Reverse)))))
  (Utilities|FlowControl|Gate false
    (:Exit
      (CallFunction|CloseDoorSound)
      (Utilities|FlowControl|Sequence
        (:then_0
          (Timeline|FadeGlint Reverse))
        (:then_1
          (Timeline|Timeline_0 Reverse))))))

(event Custom|Open%20Door%20Sound)

(event Custom|Close%20Door%20Sound)

(fn SetBattlePassVisualState
  (params (Purchased :type Boolean))
  (bind _spotlight3 (Variables|Default|GetSpotLight3))
  (bind _spotlight2 (Variables|Default|GetSpotLight2))
  (bind _mid_1 (Variables|Default|GetMID_1))
  (bind _spotlight (Variables|Default|GetSpotLight))
  (bind _mid_lights (Variables|Default|GetMID_Lights))
  (if Purchased
    (Rendering|Material|SetScalarParameterValue _mid_1 "Chrome" 0.0 :!class /Script/Engine.MaterialInstanceDynamic)
    (Rendering|SetVisibility _spotlight true false)
    (Rendering|Material|SetScalarParameterValue _mid_lights "Brightness" 1.0 :!class /Script/Engine.MaterialInstanceDynamic)
    (Rendering|Components|Light|SetLightColor _spotlight2 "(R=1.000000,G=0.132868,B=0.000000,A=1.000000)" true)
    (Rendering|Components|Light|SetLightColor _spotlight3 "(R=1.000000,G=0.318547,B=0.000000,A=1.000000)" true)
    (Development|SetHiddeninGame (Variables|Default|GetDoor_Glint) true false)
    (else
      (Rendering|Material|SetScalarParameterValue _mid_1 "Chrome" 1.0 :!class /Script/Engine.MaterialInstanceDynamic)
      (Rendering|SetVisibility _spotlight false false)
      (Rendering|Material|SetScalarParameterValue _mid_lights "Brightness" 0.01 :!class /Script/Engine.MaterialInstanceDynamic)
      (Rendering|Components|Light|SetLightColor _spotlight2 "(R=0.172394,G=0.243582,B=0.364583,A=1.000000)" true)
      (Rendering|Components|Light|SetLightColor _spotlight3 "(R=0.172394,G=0.243582,B=0.364583,A=1.000000)" true))))

(fn ConstructionScript
  (bind _s_elevator_door_bp12 (Variables|Default|GetS_Elevator_Door_BP12))
  (|CallParentFunction)
  (Utilities|FlowControl|Sequence
    (:then_0
      (bind _returnvalue (Rendering|Material|CreateDynamicMaterialInstance _s_elevator_door_bp12 0 0 "None" :!class /Script/Engine.PrimitiveComponent))
      (Variables|Default|SetMID_1 _returnvalue)
      (bind _returnvalue_2 (Rendering|Material|CreateDynamicMaterialInstance _s_elevator_door_bp12 1 0 "None" :!class /Script/Engine.PrimitiveComponent))
      (Variables|Default|SetMID_2 _returnvalue_2)
      (bind _returnvalue_3 (Rendering|Material|CreateDynamicMaterialInstance (Variables|Default|GetS_BattlePassStar_BP12) 0 0 "None" :!class /Script/Engine.PrimitiveComponent))
      (Variables|Default|SetMID_Battle_1 _returnvalue_3)
      (bind _returnvalue_4 (Rendering|Material|CreateDynamicMaterialInstance (Variables|Default|GetSM_Vent_Light) 1 "/BRCosmetics/Environments/FrontEnd/Materials/BP12/M_BP12_SimpleEmissive_BPDoor.M_BP12_SimpleEmissive_BPDoor" "None" :!class /Script/Engine.PrimitiveComponent))
      (bind _output_get (Variables|Default|SetMID_Lights _returnvalue_4))
      (Rendering|Material|SetMaterial (Variables|Default|GetSM_Vent_Light2) 1 _output_get)
      (bind _returnvalue_5 (Rendering|Material|CreateDynamicMaterialInstance (Variables|Default|GetS_BattlePassDoor_Frame_BP12) 0 0 "None" :!class /Script/Engine.PrimitiveComponent))
      (Variables|Default|SetMID_Frame _returnvalue_5))
    (:then_1
      (CallFunction|SetBattlePassVisualState (Variables|Default|GetDebugPurchased)))))
