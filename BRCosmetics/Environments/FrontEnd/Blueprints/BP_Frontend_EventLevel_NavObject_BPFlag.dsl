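; Front-end event-level navigation flag that shows whether the player owns the season Battle Pass.
; Construction creates dynamic materials for flag and glow plane, previewing state from a debug boolean.
; Begin play, player init and inventory updates check season book purchase, switching flag Chrome and gold or silver spotlights.
; Clicking plays a timeline driving the OpenMouseOver material parameter, then reverses it after a short delay.
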(event Collision|EventActorBeginOverlap
  (params (OtherActor :type Actor%20Object%20Reference))
  (|CallParentFunction OtherActor) ; This node is disabled and will not be called.\nDrag off pins to build functionality.
)

(event EventTick
  (params (DeltaSeconds :type Float%20%28single-precision%29))
  (|CallParentFunction DeltaSeconds) ; This node is disabled and will not be called.\nDrag off pins to build functionality.
)

(event EventBeginPlay
  (|CallParentFunction)
  (Utilities|FlowControl|Sequence
    (:then_0
      ;[1 1 1] Check for battlepass purchased and execute visual updates
      (EventDispatchers|BindEventtoOnInventoryUpdated (delegate Custom|UpdateVisuals)))
    (:then_1
      (CallFunction|SetBattlePassVisualState (CallFunction|HasPurchasedSeasonBook)))))

(event Custom|UpdateVisuals
  ;[1 1 1] Check for battlepass purchased and execute visual updates
  (CallFunction|SetBattlePassVisualState (CallFunction|HasPurchasedSeasonBook)))

(event EventOnInitializeForPlayer
  ;[1 1 1] Check for battlepass purchased and execute visual updates
  (CallFunction|SetBattlePassVisualState (CallFunction|HasPurchasedSeasonBook)))

(event MouseInput|EventActorBeginCursorOver)

(event MouseInput|EventActorEndCursorOver)

(event MouseInput|EventActorOnClicked
  (params (ButtonPressed :type Key%20Structure))
  (Utilities|FlowControl|Sequence
    (:then_0
      (bind (_direction _newtrack_0 _timeinxeconds) (Timeline|Timeline_0 Play)
        (:Update
          (Variables|Default|SetTimeinSeconds _timeinxeconds)
          (Rendering|Material|SetScalarParameterValue (Variables|Default|GetMID_1) "OpenMouseOver" _newtrack_0 :!class /Script/Engine.MaterialInstanceDynamic))))
    (:then_1
      _)
    (:then_2
      (Utilities|FlowControl|Delay 1.5)
      (Utilities|FlowControl|Sequence
        (:then_0
          _)
        (:then_1
          (Timeline|Timeline_0 Reverse)))))
  (Utilities|FlowControl|Gate false
    (:Exit
      (CallFunction|CloseDoorSound)
      (Utilities|FlowControl|Sequence
        (:then_0
          (bind (_direction_2 _newtrack_0_2) (Timeline|FadeGlint Reverse)
            (:Update
              (Rendering|Material|SetScalarParameterValueonMaterials (Variables|Default|GetDoor_Glint) "Opacity" _newtrack_0_2))))
        (:then_1
          (Timeline|Timeline_0 Reverse))))))

(event Custom|Open%20Door%20Sound)

(event Custom|Close%20Door%20Sound)

(fn SetBattlePassVisualState
  (params (Purchased :type Boolean))
  (bind _spotlight_gold (Variables|Default|GetSpotLight_gold))
  (bind _spotlight_silver (Variables|Default|GetSpotLight_Silver))
  (bind _mid_1 (Variables|Default|GetMID_1))
  (if Purchased
    (Rendering|Material|SetScalarParameterValue _mid_1 "Chrome" 0.0 :!class /Script/Engine.MaterialInstanceDynamic)
    (Rendering|SetVisibility _spotlight_gold true false)
    (Rendering|SetVisibility _spotlight_silver false false)
    (else
      (Rendering|Material|SetScalarParameterValue _mid_1 "Chrome" 1.0 :!class /Script/Engine.MaterialInstanceDynamic)
      (Rendering|SetVisibility _spotlight_gold false false)
      (Rendering|SetVisibility _spotlight_silver true false))))

(fn ConstructionScript
  (|CallParentFunction)
  (Utilities|FlowControl|Sequence
    (:then_0
      (bind _returnvalue (Rendering|Material|CreateDynamicMaterialInstance (Variables|Default|GetFlag) 0 0 "None" :!class /Script/Engine.PrimitiveComponent))
      (Variables|Default|SetMID_1 _returnvalue)
      (bind _returnvalue_2 (Rendering|Material|CreateDynamicMaterialInstance (Variables|Default|Get1x1plane) 0 0 "None" :!class /Script/Engine.PrimitiveComponent))
      (Variables|Default|SetMID%20GLOW _returnvalue_2))
    (:then_1
      (CallFunction|SetBattlePassVisualState (Variables|Default|GetDebugPurchased)))))
