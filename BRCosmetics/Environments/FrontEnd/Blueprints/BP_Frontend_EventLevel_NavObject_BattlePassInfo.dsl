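; Front-end navigation actor for the Battle Pass door, showing level, XP progress and the next reward on monitors and widget panels.
; Construction creates dynamic materials for the XP bar and both monitors and previews the purchased look from a debug flag.
; Battle Pass init sets bar progress, feeds both widgets and async-loads the reward image onto the right monitor.
; Player init and inventory updates recolor the left monitor and restyle its panel by season book ownership.
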(event EventBeginPlay
  (|CallParentFunction)
  (Utilities|FlowControl|Sequence
    (:then_1
      (EventDispatchers|BindEventtoOnInventoryUpdated (EventDispatchers|CreateEvent :!event Custom|UpdateVisuals)))))

(event EventOnInitializeForPlayer
  (CallFunction|UpdateVisuals))

(event Custom|UpdateVisuals
  (CallFunction|SetBattlePassVisuals (CallFunction|HasPurchasedSeasonBook)))

(event Collision|EventActorBeginOverlap
  (params (OtherActor :type Actor%20Object%20Reference))
  (|CallParentFunction OtherActor) ; This node is disabled and will not be called.\nDrag off pins to build functionality.
)

(event EventTick
  (params (DeltaSeconds :type Float%20%28single-precision%29))
  (|CallParentFunction DeltaSeconds) ; This node is disabled and will not be called.\nDrag off pins to build functionality.
)

(event EventInitForBattlePassInfo
  (params (CurrentLevel :type Integer) (CurrentXp :type Integer) (XpToLevel :type Integer) (NextReward :type Fort%20Item%20Definition%20Object%20Reference))
  (Rendering|Material|SetScalarParameterValue (Variables|Default|GetXpBarMid) "Progress" (/ (Math|Conversions|ToFloat(Integer) CurrentXp) (Math|Conversions|ToFloat(Integer) XpToLevel)) :!class /Script/Engine.MaterialInstanceDynamic)
  (bind _asbattle_pass_door_xp_progress (Utilities|Casting|CastToBattlePass_Door_XP_Progress (UserInterface|GetUserWidgetObject (Variables|Default|GetXpValues))))
  (Class|BattlePassDoorXPProgress|InitXpValues _asbattle_pass_door_xp_progress CurrentXp XpToLevel (+ CurrentLevel 1))
  (bind _asbattle_pass_door_left_panel (Utilities|Casting|CastToBattlePass_Door_LeftPanel (UserInterface|GetUserWidgetObject (Variables|FortEventLevelNavigationActor|GetWidgetComponent_NavigationWidget))))
  (Class|BattlePassDoorLeftPanel|InitLevel _asbattle_pass_door_left_panel CurrentLevel)
  (bind _object (Utilities|AsyncLoadAsset (Item|GetLargePreviewImage NextReward))
    (:Completed
      (bind _astexture (Utilities|Casting|CastToTexture _object))
      (Rendering|Material|SetTextureParameterValue (Variables|Default|GetBPRewardsMid) "RewardImage" _astexture))))

(event EventOnUpdateWidgetVisibility
  (params (PlayerController :type Fort%20Player%20Controller%20Object%20Reference) (bEnabled :type Boolean))
  (bind _xpvalues (Variables|Default|GetXpValues))
  (bind _returnvalue (UserInterface|GetUserWidgetObject _xpvalues))
  (Player|SetOwningPlayer _returnvalue PlayerController)
  (Widget|SetVisibility _returnvalue (select bEnabled "Visible" "Hidden")))

(fn SetBattlePassVisuals
  (params (Purchased :type Boolean))
  (bind _mid_leftmonitor (Variables|Default|GetMID_LeftMonitor))
  (if Purchased
    (Rendering|Material|SetVectorParameterValue _mid_leftmonitor "Color1" "(R=0.235000,G=0.035696,B=0.017578,A=1.000000)" :!class /Script/Engine.MaterialInstanceDynamic)
    (Rendering|Material|SetVectorParameterValue _mid_leftmonitor "Color2" "(R=3.000000,G=0.833000,B=0.045000,A=1.000000)" :!class /Script/Engine.MaterialInstanceDynamic)
    (else
      (Rendering|Material|SetVectorParameterValue _mid_leftmonitor "Color1" "(R=0.014444,G=0.026241,B=0.149960,A=1.000000)" :!class /Script/Engine.MaterialInstanceDynamic)
      (Rendering|Material|SetVectorParameterValue _mid_leftmonitor "Color2" "(R=0.136875,G=0.980937,B=1.825000,A=1.000000)" :!class /Script/Engine.MaterialInstanceDynamic)))
  (bind _asbattle_pass_door_left_panel (Utilities|Casting|CastToBattlePass_Door_LeftPanel (UserInterface|GetUserWidgetObject (Variables|FortEventLevelNavigationActor|GetWidgetComponent_NavigationWidget))))
  (Class|BattlePassDoorLeftPanel|SetDoorPanelStyle _asbattle_pass_door_left_panel Purchased))

(fn ConstructionScript
  (|CallParentFunction)
  (Utilities|FlowControl|Sequence
    (:then_0
      (bind _returnvalue (Rendering|Material|CreateDynamicMaterialInstance (Variables|Default|GetXpBar) 0 "/BRCosmetics/Environments/FrontEnd/Materials/BP12/M_Progressbar.M_Progressbar" "None" :!class /Script/Engine.PrimitiveComponent))
      (Variables|Default|SetXpBarMid _returnvalue)
      (bind _returnvalue_2 (Rendering|Material|CreateDynamicMaterialInstance (Variables|Default|GetRightMonitor) 1 "/Game/Environments/FrontEnd/Materials/BP12/M_Monitor_BP12.M_Monitor_BP12" "None" :!class /Script/Engine.PrimitiveComponent))
      (Variables|Default|SetBPRewardsMid _returnvalue_2)
      (bind _returnvalue_3 (Rendering|Material|CreateDynamicMaterialInstance (Variables|Default|GetLeftMonitor) 1 0 "None" :!class /Script/Engine.PrimitiveComponent))
      (Variables|Default|SetMID_LeftMonitor _returnvalue_3))
    (:then_1
      (CallFunction|SetBattlePassVisuals (Variables|Default|GetDeubgPurchased)))))
