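; Front-end lobby nav object for the Season 14 battle pass room monitor, showing a gold or silver logo based on whether the season book is purchased.
; It checks purchase state on BeginPlay, on player initialization and whenever the inventory-updated event fires.
; SetBattlePassVisualState creates a dynamic material instance on the Plane component and swaps its logo overlay texture, creating a fresh instance on each call.
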(event EventBeginPlay
  ;[0 0 0 1 12] Sets Visual Style Based on the BP Purchase State
  (|CallParentFunction)
  (Utilities|FlowControl|Sequence
    (:then_0
      (EventDispatchers|BindEventtoOnInventoryUpdated (delegate Custom|UpdateVisuals)))
    (:then_1
      (CallFunction|SetBattlePassVisualState (CallFunction|HasPurchasedSeasonBook)))))

(event Custom|UpdateVisuals
  ;[0 0 0 1 12] Sets Visual Style Based on the BP Purchase State
  (CallFunction|SetBattlePassVisualState (CallFunction|HasPurchasedSeasonBook)))

(event EventOnInitializeForPlayer
  ;[0 0 0 1 12] Sets Visual Style Based on the BP Purchase State
  (CallFunction|SetBattlePassVisualState (CallFunction|HasPurchasedSeasonBook)))

(fn SetBattlePassVisualState
  (params (bIsPurchased :type Boolean))
  (bind _returnvalue (Rendering|Material|CreateDynamicMaterialInstance (Variables|Default|GetPlane) 0 0 "None" :!class /Script/Engine.PrimitiveComponent))
  (Variables|Default|SetMID_BPScreen _returnvalue)
  (if bIsPurchased
    (Rendering|Material|SetTextureParameterValue (Variables|Default|GetMID_BPScreen) "BP_LogoOverlay" "/BRCosmetics/Environments/FrontEnd/Textures/BP14/T_S14_BPRoomMonitor_GoldBP_02-Logo.T_S14_BPRoomMonitor_GoldBP_02-Logo")
    (else
      (Rendering|Material|SetTextureParameterValue (Variables|Default|GetMID_BPScreen) "BP_LogoOverlay" "/BRCosmetics/Environments/FrontEnd/Textures/BP14/T_S14_BPRoomMonitor_SilverBP-Logo.T_S14_BPRoomMonitor_SilverBP-Logo"))))

(fn ConstructionScript)
