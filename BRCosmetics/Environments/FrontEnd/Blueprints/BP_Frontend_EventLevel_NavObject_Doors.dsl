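; Frontend event-level door nav object that animates and plays sound on mouse hover and click.
; Construction creates dynamic material instances on the elevator door mesh, and a timeline drives the "OpenMouseOver" parameter to animate it through the material.
; Hover plays the timeline, gated cursor exit reverses it, and click plays it then reverses after a delay.
; Armory or Barracks sounds, chosen by override-material validity, cross-fade from timeline-based offsets.
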(event EventBeginPlay
  (|CallParentFunction))

(event Collision|EventActorBeginOverlap
  (params (OtherActor :type Actor%20Object%20Reference))
  (|CallParentFunction OtherActor) ; This node is disabled and will not be called.\nDrag off pins to build functionality.
)

(event EventTick
  (params (DeltaSeconds :type Float%20%28single-precision%29))
  (|CallParentFunction DeltaSeconds) ; This node is disabled and will not be called.\nDrag off pins to build functionality.
)

(event MouseInput|EventActorBeginCursorOver
  (CallFunction|OpenDoorSound)
  (bind (_direction _newtrack_0 _timeinseconds) (Timeline|Timeline_0 Play)
    (:Update
      (Variables|Default|Settimeinseconds _timeinseconds)
      (Rendering|Material|SetScalarParameterValue (Variables|Default|GetMID_1) "OpenMouseOver" _newtrack_0 :!class /Script/Engine.MaterialInstanceDynamic))))

(event MouseInput|EventActorEndCursorOver
  (Utilities|FlowControl|Gate false
    (:Exit
      (CallFunction|CloseDoorSound)
      (Timeline|Timeline_0 Reverse))))

(event MouseInput|EventActorOnClicked
  (params (ButtonPressed :type Key%20Structure))
  (Utilities|FlowControl|Sequence
    (:then_0
      (Timeline|Timeline_0 Play))
    (:then_1
      _)
    (:then_2
      (Utilities|FlowControl|Delay 1.5)
      (Utilities|FlowControl|Sequence
        (:then_0
          _)
        (:then_1
          (Timeline|Timeline_0 Reverse)))))
  (Utilities|FlowControl|Gate false
    (:Exit
      (CallFunction|CloseDoorSound)
      (Timeline|Timeline_0 Reverse))))

(event Custom|Open%20Door%20Sound
  (bind _returnvalue (Audio|SpawnSoundatLocation (select (Utilities|IsValid (Variables|Default|GetElevatorMaterial%20override) :!class /Script/Engine.KismetSystemLibrary) "/BRCosmetics/Sounds/Other/Fort_UI_Sounds/S12_Frontend/UI_Armory_Door_Open_Cue.UI_Armory_Door_Open_Cue" "/BRCosmetics/Sounds/Other/Fort_UI_Sounds/S12_Frontend/UI_Barracks_Door_Open_Cue.UI_Barracks_Door_Open_Cue") (Transformation|GetActorLocation)))
  (Variables|Default|SetOpeningSound _returnvalue)
  (Audio|Components|Audio|FadeIn (Variables|Default|GetOpeningSound) 0.2 1.0 (Variables|Default|GetTimeinSeconds) "Linear" :!class /Script/Engine.AudioComponent)
  (Audio|Components|Audio|FadeOut (Variables|Default|GetClosingSound) 0.5 0.0 "Linear" :!class /Script/Engine.AudioComponent))

(event Custom|Close%20Door%20Sound
  (bind _returnvalue (Audio|SpawnSoundatLocation (select (Utilities|IsValid (Variables|Default|GetElevatorMaterial%20override) :!class /Script/Engine.KismetSystemLibrary) "/BRCosmetics/Sounds/Other/Fort_UI_Sounds/S12_Frontend/UI_Armory_Door_Close_Cue.UI_Armory_Door_Close_Cue" "/BRCosmetics/Sounds/Other/Fort_UI_Sounds/S12_Frontend/UI_Barracks_Door_Close_Cue.UI_Barracks_Door_Close_Cue") (Transformation|GetActorLocation)))
  (Variables|Default|SetClosingSound _returnvalue)
  (Audio|Components|Audio|FadeIn (Variables|Default|GetClosingSound) 0.5 1.0 (- 2.0 (Variables|Default|GetTimeinSeconds)) "Linear" :!class /Script/Engine.AudioComponent)
  (Audio|Components|Audio|FadeOut (Variables|Default|GetOpeningSound) 0.5 0.0 "Linear" :!class /Script/Engine.AudioComponent))

(fn ConstructionScript
  (bind _s_elevator_door_bp12 (Variables|Default|GetS_Elevator_Door_BP12))
  (|CallParentFunction)
  (bind _returnvalue (Rendering|Material|CreateDynamicMaterialInstance _s_elevator_door_bp12 0 (select (Utilities|IsValid (Variables|Default|GetElevatorMaterial%20override) :!class /Script/Engine.KismetSystemLibrary) (Variables|Default|GetElevatorMaterial%20override) (Rendering|Material|GetMaterial _s_elevator_door_bp12 0)) "None" :!class /Script/Engine.PrimitiveComponent))
  (bind _output_get (Variables|Default|SetMID_1 _returnvalue))
  (Rendering|Material|SetMaterial _s_elevator_door_bp12 0 _output_get)
  (bind _returnvalue_2 (Rendering|Material|CreateDynamicMaterialInstance _s_elevator_door_bp12 1 0 "None" :!class /Script/Engine.PrimitiveComponent))
  (Variables|Default|SetMID_2 _returnvalue_2))
