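; Front-end map table navigation actor that shows a minimap texture and discovery mask on its Apollo map.
; On BeginPlay, if both texture arrays are filled, it assigns a dynamic minimap material and applies textures, reapplying them when active events change.
; Textures are picked by clamped GetMinimapIndex, which returns 0 here.
; Both async load in parallel, and a wait node gates setting the material parameters.
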(event EventBeginPlay
  (|CallParentFunction)
  (if (CallFunction|HasDynamicMinimapTextures)
    (bind _returnvalue (Rendering|Material|CreateDynamicMaterialInstance (Variables|Default|GetMinimapMaterial) "None" "Transient" :!class /Script/Engine.KismetMaterialLibrary))
    (bind _output_get (Variables|Default|SetMinimapMID _returnvalue))
    (Rendering|Material|SetMaterial (Variables|FortNavigationActor_MapTable|GetApolloMap) 0 _output_get)
    (CallFunction|ApplyMinimapTextures)))

(event Collision|EventActorBeginOverlap
  (params (OtherActor :type Actor%20Object%20Reference))
  (|CallParentFunction OtherActor) ; This node is disabled and will not be called.\nDrag off pins to build functionality.
)

(event EventTick
  (params (DeltaSeconds :type Float%20%28single-precision%29))
  (|CallParentFunction DeltaSeconds) ; This node is disabled and will not be called.\nDrag off pins to build functionality.
)

(event EventOnActiveEventsChanged
  (if (CallFunction|HasDynamicMinimapTextures)
    (CallFunction|ApplyMinimapTextures)))

(event Custom|ApplyMinimapTextures
  (bind _minimapmid (Variables|Default|GetMinimapMID))
  (bind (_minimaptexture _minimapdiscoverymask) (CallFunction|GetMinimapTextures))
  (bind _object (Utilities|AsyncLoadAsset _minimaptexture)
    (:then
      (bind _object_2 (Utilities|AsyncLoadAsset _minimapdiscoverymask)
        (:then
          _)
        (:Completed
          _)))
    (:Completed
      _))
  (Utilities|FlowControl|WaitforBothExecutions true
    (:Then
      (bind _astexture_2d (Utilities|Casting|CastToTexture2D _object))
      (bind _astexture_2d_2 (Utilities|Casting|CastToTexture2D _object_2))
      (Utilities|IsValid _minimapmid
        (:Is%20Valid
          (Rendering|Material|SetTextureParameterValue _minimapmid "MinimapTexture" _astexture_2d)
          (Rendering|Material|SetTextureParameterValue _minimapmid "MinimapDiscoveryMask" _astexture_2d_2))))))

(fn GetMinimapIndex
  (return 0))

(fn GetMinimapTextures
  (if (CallFunction|HasDynamicMinimapTextures)
    (return (Utilities|Array|Get(acopy) (Variables|Default|GetMinimapTextures) (Math|Integer|Clamp(Integer) (CallFunction|GetMinimapIndex) 0 (- (Utilities|Array|Length (Variables|Default|GetMinimapTextures)) 1))) (Utilities|Array|Get(acopy) (Variables|Default|GetMinimapDiscoveryMasks) (Math|Integer|Clamp(Integer) (CallFunction|GetMinimapIndex) 0 (- (Utilities|Array|Length (Variables|Default|GetMinimapDiscoveryMasks)) 1))) (CallFunction|GetMinimapIndex))
    (else
      (return "" "" -1))))

(fn HasDynamicMinimapTextures
  (return (and (Utilities|IsArrayNonEmpty (Variables|Default|GetMinimapTextures)) (Utilities|IsArrayNonEmpty (Variables|Default|GetMinimapDiscoveryMasks)))))

(fn ConstructionScript
  (|CallParentFunction))
