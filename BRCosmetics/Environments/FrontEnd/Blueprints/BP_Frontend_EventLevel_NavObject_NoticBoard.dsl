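; Front-end notice board nav object that shows a hint, map and pots according to local player progress in two TvB quests.
; On BeginPlay it awaits the async quest query: an incomplete Maproom quest shows only the hint, a complete one reveals the map.
; A complete Pots quest hides the map, otherwise each finished objective hides its mapped component, with debug prints.
; The construction script stretches five spline meshes between paired pins.
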(event EventBeginPlay
  (bind _tvb_quests_to_check (Variables|Default|GetTvB_Quests_To_Check))
  (|CallParentFunction)
  (Utilities|Array|Add _tvb_quests_to_check (Variables|Default|GetTvB_02_Pots))
  (Utilities|Array|Add _tvb_quests_to_check (Variables|Default|GetTvB_01_Maproom))
  (bind (_owningquestmanager _bfoundallquestsonprofile) (Quests|CurrentQuestsReadyAsync (Quests|GetQuestManager (Fortnite|GetFirstLocalFortPlayerController true) "Invalid") (Variables|Default|GetTvB_Quests_To_Check))
    (:OnCurrentQuestsReady
      (if _bfoundallquestsonprofile
        (bind (_questisvalid _questcompleted) (Default|AthenaCheckQuestandObjectiveCompletion (Fortnite|GetFirstLocalFortPlayerController true) (Variables|Default|GetTvB_01_Maproom) (Variables|Default|GetMapRoom_ObjName)))
        (if _questcompleted
          (Development|PrintString "MapQuestValid: YES" false false "(R=0.154764,G=1.000000,B=0.000000,A=1.000000)" 1e+01)
          (CallFunction|TvBShowOnlyMapandPots)
          (bind (_questisvalid_2 _questcompleted_2) (Default|AthenaCheckQuestandObjectiveCompletion (Fortnite|GetFirstLocalFortPlayerController true) (Variables|Default|GetTvB_02_Pots)))
          (if _questcompleted_2
            (CallFunction|TvBHideAll)
            (else
              (for (_array_element _array_index) (Variables|Default|Getpots_ObjName_all)
                (bind (_questisvalid_3 _questcompleted_3 _objectivecompleted) (Default|AthenaCheckQuestandObjectiveCompletion (Fortnite|GetFirstLocalFortPlayerController true) (Variables|Default|GetTvB_02_Pots) _array_element))
                (if _objectivecompleted
                  (switch int _array_index
                    (:0
                      (CallFunction|HidePot01)
                      (Development|PrintString "POT 1: Hidden" false false "(R=0.000000,G=0.660000,B=1.000000,A=1.000000)" 15.0))
                    (:1
                      (CallFunction|HidePot02)
                      (Development|PrintString "POT 2: Hidden" false false "(R=0.000000,G=0.660000,B=1.000000,A=1.000000)" 15.0))
                    (:2
                      (CallFunction|HidePot03)
                      (Development|PrintString "POT 3: Hidden" false false "(R=0.000000,G=0.660000,B=1.000000,A=1.000000)" 15.0))
                    (:3
                      (CallFunction|HidePot04)
                      (Development|PrintString "POT 4: Hidden" false false "(R=0.000000,G=0.660000,B=1.000000,A=1.000000)" 15.0))
                    (:4
                      (CallFunction|HidePot05)
                      (Development|PrintString "POT 5: Hidden" false false "(R=0.000000,G=0.660000,B=1.000000,A=1.000000)" 15.0)))
                  (else
                    (Development|PrintString (Utilities|String|Append (Utilities|String|ToString(Name) _array_element) " NOT COMPLETE ") false false "(R=1.000000,G=0.020315,B=0.000000,A=1.000000)" 15.0))))))
          (else
            (Development|PrintString "MapQuestValid: NO" false false "(R=1.000000,G=0.009094,B=0.000000,A=1.000000)" 1e+01)
            (CallFunction|TvBShowOnlyHint)))))))

(event Collision|EventActorBeginOverlap
  (params (OtherActor :type Actor%20Object%20Reference))
  (|CallParentFunction OtherActor) ; This node is disabled and will not be called.\nDrag off pins to build functionality.
)

(event EventTick
  (params (DeltaSeconds :type Float%20%28single-precision%29))
  (|CallParentFunction DeltaSeconds) ; This node is disabled and will not be called.\nDrag off pins to build functionality.
)

(fn Hide_Pot_01
  (Development|SetHiddeninGame (Variables|Default|GetS_ObjectiveBoard_Pin_Honey_1) true true))

(fn Hide_Pot_02
  (Development|SetHiddeninGame (Variables|Default|GetSplineMesh_2) true true))

(fn Hide_Pot_03
  (Development|SetHiddeninGame (Variables|Default|GetS_ObjectiveBoard_TvB_Pot_3) true true))

(fn Hide_Pot_04
  (Development|SetHiddeninGame (Variables|Default|GetS_ObjectiveBoard_TvB_Pot_4) true true))

(fn Hide_Pot_05
  (Development|SetHiddeninGame (Variables|Default|GetS_ObjectiveBoard_TvB_Pot_5) true true))

(fn TvB_HideAll
  (Development|SetHiddeninGame (Variables|Default|GetS_ObjectiveBoard_TvB_Map) true true))

(fn TvB_ShowOnlyHint
  (Development|SetHiddeninGame (Variables|Default|GetS_ObjectiveBoard_TvB_Hint) false true)
  (Development|SetHiddeninGame (Variables|Default|GetS_ObjectiveBoard_TvB_Map) true true))

(fn TvB_ShowOnlyMapAndPots
  (Development|SetHiddeninGame (Variables|Default|GetS_ObjectiveBoard_TvB_Map) false true)
  (Development|SetHiddeninGame (Variables|Default|GetS_ObjectiveBoard_TvB_Hint) true true))

(fn UpdateAllSplines
  (bind _returnvalue (Math|Vector|MakeVector 0.0 5.0))
  (CallFunction|UpdateSplines (Variables|Default|GetSplineMesh_3) (Variables|Default|GetS_ObjectiveBoard_Pin_Honey_3) (Variables|Default|GetS_ObjectiveBoard_Pin_Honey_3E) _returnvalue)
  (CallFunction|UpdateSplines (Variables|Default|GetSplineMesh_4) (Variables|Default|GetS_ObjectiveBoard_Pin_Honey_4) (Variables|Default|GetS_ObjectiveBoard_Pin_Honey_4E) _returnvalue)
  (CallFunction|UpdateSplines (Variables|Default|GetSplineMesh_2) (Variables|Default|GetS_ObjectiveBoard_Pin_Honey_2) (Variables|Default|GetS_ObjectiveBoard_Pin_Honey_2E) _returnvalue)
  (CallFunction|UpdateSplines (Variables|Default|GetSplineMesh_1) (Variables|Default|GetS_ObjectiveBoard_Pin_Honey_1) (Variables|Default|GetS_ObjectiveBoard_Pin_Honey_1E) _returnvalue)
  (CallFunction|UpdateSplines (Variables|Default|GetSplineMesh_5) (Variables|Default|GetS_ObjectiveBoard_Pin_Honey_5) (Variables|Default|GetS_ObjectiveBoard_Pin_Honey_5E) _returnvalue))

(fn UpdateSplines
  (params (Spline :type Spline%20Mesh%20Component%20Object%20Reference) (Start :type Static%20Mesh%20Component%20Object%20Reference) (End :type Static%20Mesh%20Component%20Object%20Reference) (Offset :type Vector))
  (SplineMesh|SetStartandEnd Spline (+ (Variables|/Script/Engine.SceneComponent|GetRelativeLocation Start) Offset) (- (+ (Variables|/Script/Engine.SceneComponent|GetRelativeLocation End) Offset) (+ (Variables|/Script/Engine.SceneComponent|GetRelativeLocation Start) Offset)) (+ (Variables|/Script/Engine.SceneComponent|GetRelativeLocation End) Offset) (- (+ (Variables|/Script/Engine.SceneComponent|GetRelativeLocation End) Offset) (+ (Variables|/Script/Engine.SceneComponent|GetRelativeLocation Start) Offset)) true :!class /Script/Engine.SplineMeshComponent))

(fn ConstructionScript
  (|CallParentFunction)
  (CallFunction|UpdateAllSplines))
