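; Frontend event-level navigation object Blueprint that shows or hides a mannequin mesh according to navigation visibility checks.
; When the custom visibility check passes, it clears any pending timer and schedules the mannequin to show after a short delay.
; When the check fails, it does the same but hides it after a longer delay, so quick state changes cancel pending switches instead of flickering.
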(event EventBeginPlay
  (|CallParentFunction))

(event Collision|EventActorBeginOverlap
  (params (OtherActor :type Actor%20Object%20Reference))
  (|CallParentFunction OtherActor) ; This node is disabled and will not be called.\nDrag off pins to build functionality.
)

(event EventTick
  (params (DeltaSeconds :type Float%20%28single-precision%29))
  (|CallParentFunction DeltaSeconds) ; This node is disabled and will not be called.\nDrag off pins to build functionality.
)

(event HandleCustomPassed(FortNavigationVisibility)
  (CallFunction|ClearVisTimer)
  (bind _returnvalue (Utilities|Time|SetTimerbyEvent (delegate Custom|ShowMannequin) 0.2))
  (Variables|Default|SetVisTimerHandle _returnvalue))

(event Custom|ShowMannequin
  (Rendering|SetVisibility (Variables|Default|GetMannequin) true false))

(event HandleCustomFailed(FortNavigationVisibility)
  (CallFunction|ClearVisTimer)
  (bind _returnvalue (Utilities|Time|SetTimerbyEvent (delegate Custom|HideMannequin) 1.0))
  (Variables|Default|SetVisTimerHandle _returnvalue))

(event Custom|HideMannequin
  (Rendering|SetVisibility (Variables|Default|GetMannequin) false false))

(event Custom|ClearVisTimer
  (if (Utilities|Time|IsValidTimerHandle (Variables|Default|GetVisTimerHandle))
    (Utilities|Time|ClearandInvalidateTimerbyHandle (Variables|Default|GetVisTimerHandle))))

(fn ConstructionScript
  (|CallParentFunction))
