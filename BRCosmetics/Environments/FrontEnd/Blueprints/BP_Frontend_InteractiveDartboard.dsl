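; Interactive frontend dartboard that throws one of three darts per allowed special interaction, with sound and particles.
; BeginPlay caches dart start locations; a counter selects the dart, and its timeline lerps it to its target, always starting from Dart1's origin.
; Interaction locks during each throw and unlocks after the first two only.
; The third zeroes the counter and optionally starts a timer that restores dart positions without unlocking interaction.
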(event EventBeginPlay
  ;[1 1 1] EVENT - BEGINPLAY
  (Variables|Default|SetDart1Source (Variables|/Script/Engine.SceneComponent|GetRelativeLocation (Variables|Default|GetDart1)))
  (Variables|Default|SetDart2Source (Variables|/Script/Engine.SceneComponent|GetRelativeLocation (Variables|Default|GetDart2)))
  (Variables|Default|SetDart3Source (Variables|/Script/Engine.SceneComponent|GetRelativeLocation (Variables|Default|GetDart3))))

(event EventHandleObjectSpecialInteraction
  (params (ActionTag :type Gameplay%20Tag%20Structure) (CameraTag :type Gameplay%20Tag%20Structure))
  (bind _dartint (Variables|Default|GetDartInt))
  (if (Variables|Default|GetIsInteractAllowed)
    (Variables|Default|SetIsInteractAllowed false)
    (Utilities|FlowControl|Sequence
      (:then_0
        (switch int (Variables|Default|GetDartInt)
          (:0
            (Math|Integer|IncrementInt _dartint
              (:%20%20
                (Components|Activation|Activate (Variables|Default|GetPFX_Dart) true)
                (Audio|SpawnSound2D "/BRCosmetics/Sounds/Other/Fort_UI_Sounds/S12_Frontend/UI_Meowscles_Dart_Cue.UI_Meowscles_Dart_Cue")
                (bind (_direction _flightfloat) (Timeline|AnimateDart PlayFromStart)
                  (:Update
                    (Transformation|SetRelativeLocation (Variables|Default|GetDart1) (Math|Vector|Lerp(Vector) (Variables|Default|GetDart1Source) (Variables|/Script/Engine.SceneComponent|GetRelativeLocation (Variables|Default|GetDartTarget1)) _flightfloat) false false))
                  (:Finished
                    (Variables|Default|SetIsInteractAllowed true))
                  (:DisableFX
                    (Components|Activation|Deactivate (Variables|Default|GetPFX_Dart)))))))
          (:1
            (Math|Integer|IncrementInt _dartint
              (:%20%20
                (Components|Activation|Activate (Variables|Default|GetPFX_Dart1) true)
                (Audio|SpawnSound2D "/BRCosmetics/Sounds/Other/Fort_UI_Sounds/S12_Frontend/UI_Meowscles_Dart_Cue.UI_Meowscles_Dart_Cue")
                (bind (_direction_2 _flightfloat_2) (Timeline|Timeline_0 PlayFromStart)
                  (:Update
                    (Transformation|SetRelativeLocation (Variables|Default|GetDart2) (Math|Vector|Lerp(Vector) (Variables|Default|GetDart1Source) (Variables|/Script/Engine.SceneComponent|GetRelativeLocation (Variables|Default|GetDartTarget2)) _flightfloat_2) false false))
                  (:Finished
                    (Variables|Default|SetIsInteractAllowed true))
                  (:DisableFX
                    (Components|Activation|Deactivate (Variables|Default|GetPFX_Dart1)))))))
          (:2
            (Variables|Default|SetDartInt 0)
            (Components|Activation|Activate (Variables|Default|GetPFX_Dart2) true)
            (Audio|SpawnSound2D "/BRCosmetics/Sounds/Other/Fort_UI_Sounds/S12_Frontend/UI_Meowscles_Dart_Cue.UI_Meowscles_Dart_Cue")
            (bind (_direction_3 _flightfloat_3) (Timeline|Timeline_1 PlayFromStart)
              (:Update
                (Transformation|SetRelativeLocation (Variables|Default|GetDart3) (Math|Vector|Lerp(Vector) (Variables|Default|GetDart1Source) (Variables|/Script/Engine.SceneComponent|GetRelativeLocation (Variables|Default|GetDartTarget3)) _flightfloat_3) false false))
              (:Finished
                (if (Variables|Default|GetbUseResetTimer)
                  (Utilities|Time|SetTimerbyEvent (delegate Custom|ResetAll) 3.0)))
              (:DisableFX
                (Components|Activation|Deactivate (Variables|Default|GetPFX_Dart2)))))))
      (:then_1
        (CallFunction|CommitonSpecialInteraction)))))

(event Custom|ResetAll
  (Transformation|SetRelativeLocation (Variables|Default|GetDart1) (Variables|Default|GetDart1Source) false false)
  (Transformation|SetRelativeLocation (Variables|Default|GetDart2) (Variables|Default|GetDart2Source) false false)
  (Transformation|SetRelativeLocation (Variables|Default|GetDart3) (Variables|Default|GetDart3Source) false false))

(event Custom|Commit_OnSpecialInteraction)

(fn CanCommitNavigationRequest
  (return true))

(fn ConstructionScript)
