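; Clickable glowing star prop for the front-end environment, apparently one piece of an ordered star-sequence interaction.
; Construction builds a dynamic emissive material, twinkle particle and scale.
; A special interaction locks input, broadcasts StarClicked with its Order, and runs a timeline driving glow and spin.
; Custom events play the off-sequence tone, play success tone, burst and shine, or on exhaustion stop shine, reset glow and re-enable input.
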(event EventBeginPlay
  ;[1 1 1] EVENT - BEGINPLAY
  (Variables|Default|SetInitialRotation (Variables|/Script/Engine.SceneComponent|GetRelativeRotation (Variables|Default|GetCISFixMesh))))

(event EventHandleObjectSpecialInteraction
  (params (ActionTag :type Gameplay%20Tag%20Structure) (CameraTag :type Gameplay%20Tag%20Structure))
  (if (Variables|Default|GetIsInteractAllowed)
    (Variables|Default|SetIsInteractAllowed false)
    (Utilities|FlowControl|Sequence
      (:then_0
        ;[1 1 1]
        (Default|CallStarClicked (Variables|Default|GetOrder) self)
        (bind (_direction _glowcurve _starspin) (Timeline|MagicMel PlayFromStart)
          (:Update
            (Rendering|Material|SetVectorParameterValue (Variables|Default|GetStarMat) "Emissive" (Math|Color|LinearColor*Float (Variables|Default|GetGlowColor) _glowcurve) :!class /Script/Engine.MaterialInstanceDynamic)
            (Transformation|SetRelativeRotation (Variables|Default|GetCISFixMesh) (Math|Rotator|CombineRotators (Variables|Default|GetInitialRotation) (Math|Rotator|MakeRotator 0.0 (* _starspin (Variables|Default|GetRotationMultiplier)))) false false))))
      (:then_1
        (CallFunction|CommitonSpecialInteraction)))))

(event Custom|OffSequence
  (Audio|PlaySound2D (Variables|Default|GetActiveTone)))

(event Custom|MainSequence
  (Audio|PlaySound2D (Variables|Default|GetSuccessTone))
  (CallFunction|BigPuff)
  (Utilities|FlowControl|Delay 0.5)
  (Components|Activation|Activate (Variables|Default|GetStarShine) false))

(event Custom|BigPuff
  (Effects|SpawnEmitteratLocation (Variables|Default|GetStarExplode) (Transformation|GetWorldLocation (Variables|Default|GetStarShine)) (Transformation|GetWorldRotation (Variables|Default|GetStarShine)) "1.500000,1.500000,1.500000"))

(event Custom|Commit_OnSpecialInteraction)

(event Custom|Exhaustion
  (Utilities|FlowControl|Delay 0.5)
  (Components|Activation|Deactivate (Variables|Default|GetStarShine))
  (Rendering|Material|SetVectorParameterValue (Variables|Default|GetStarMat) "Emissive" (Math|Color|LinearColor*Float (Variables|Default|GetGlowColor) 1.0) :!class /Script/Engine.MaterialInstanceDynamic)
  (Variables|Default|SetIsInteractAllowed true))

(fn CanCommitNavigationRequest
  (return true))

(fn ConstructionScript
  (Effects|Components|ParticleSystem|SetTemplate (Variables|Default|GetStarShine) (Variables|Default|GetStarTwinkle))
  (bind _returnvalue (Rendering|Material|CreateDynamicMaterialInstance (Variables|Default|GetCISFixMesh) 0 "/BRCosmetics/Environments/FrontEnd/Materials/BP12/MI_PrincessCastle_GlowingStar_AG_FE.MI_PrincessCastle_GlowingStar_AG_FE" "None" :!class /Script/Engine.PrimitiveComponent))
  (bind _output_get (Variables|Default|SetStarMat _returnvalue))
  (Rendering|Material|SetVectorParameterValue _output_get "Emissive" (Variables|Default|GetGlowColor) :!class /Script/Engine.MaterialInstanceDynamic)
  (Transformation|SetWorldScale3D (Variables|Default|GetCISFixMesh) (Math|Conversions|ToVector(Float) (Variables|Default|GetScale))))
