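; Interactive front-end valve that spins when clicked and reports its order to listeners, apparently as part of a sequence puzzle.
; Interaction locks the valve, plays SpinForward, then starts a timer that broadcasts ValveExpired and plays SpinBack, which re-enables interaction when finished.
; Externally callable events enable a steam trickle, reset the valve, or mark success with a tinted steam burst and sound.
; Both timelines rotate the mesh's yaw by a configurable multiplier.
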(event Custom|EnableTrickle
  ;[1 1 1] Unwind
  (Components|Activation|Activate (Variables|Default|GetValidSteam) true))

(event Custom|ValveSucceed
  ;[1 1 1] Unwind
  (Utilities|Time|ClearTimerbyFunctionName 0 "RevertValve")
  (Audio|PlaySound2D (Variables|Default|GetSteamReleaseSound))
  (bind _returnvalue (Effects|SpawnEmitteratLocation :EmitterTemplate (Variables|Default|GetSteamEmitter) :Location (Transformation|GetActorLocation) :Scale "2.000000,2.000000,2.000000" :bAutoDestroy true :PoolingMethod "None" :bAutoActivateSystem true))
  (Effects|Components|ParticleSystem|SetColorParameter _returnvalue "ColorMult" (Variables|VFX|GetSteamColor))
  (Components|Activation|Deactivate (Variables|Default|GetValidSteam))
  (bind (_direction _rotation) (Timeline|SpinBack PlayFromStart)
    (:Update
      (Transformation|SetRelativeLocationAndRotation (Variables|Default|GetCISFixMesh) "0, 0, 0" (Math|Rotator|MakeRotator 0.0 (* _rotation (Variables|Default|GetRotationMultiplier))) false false))
    (:Finished
      (Variables|Default|SetIsInteractAllowed true))))

(event EventBeginPlay
  ;[1 1 1] EVENT - BEGINPLAY
  (Variables|Default|SetInitialRotation (Transformation|GetActorRotation)))

(event Custom|RevertValve
  ;[1 1 1] Unwind
  (if (not (Variables|Default|GetIsInteractAllowed))
    (Default|CallValveExpired)
    (Timeline|SpinBack PlayFromStart)))

(event Custom|ResetValve
  ;[1 1 1] Unwind
  (if (not (Variables|Default|GetIsInteractAllowed))
    (Components|Activation|Deactivate (Variables|Default|GetValidSteam))
    (Utilities|Time|ClearTimerbyFunctionName 0 "RevertValve")
    (Timeline|SpinBack PlayFromStart)))

(event EventHandleObjectSpecialInteraction
  (params (ActionTag :type Gameplay%20Tag%20Structure) (CameraTag :type Gameplay%20Tag%20Structure))
  (if (Variables|Default|GetIsInteractAllowed)
    (Variables|Default|SetIsInteractAllowed false)
    (Utilities|FlowControl|Sequence
      (:then_0
        (Audio|PlaySound2D (Variables|Default|GetTightenSound))
        ;[1 1 1] Wind up
        (Default|CallValveClicked (Variables|Default|GetOrder) self)
        (bind (_direction _rotation) (Timeline|SpinForward PlayFromStart)
          (:Update
            (Transformation|SetRelativeLocationAndRotation (Variables|Default|GetCISFixMesh) "0, 0, 0" (Math|Rotator|MakeRotator 0.0 (* _rotation (Variables|Default|GetRotationMultiplier))) false false))
          (:Finished
            (Utilities|Time|SetTimerbyFunctionName 0 "RevertValve" (Variables|Default|GetResetTimer)))))
      (:then_1
        (CallFunction|CommitonSpecialInteraction)))))

(event Custom|Commit_OnSpecialInteraction)

(fn CanCommitNavigationRequest
  (return true))

(fn ConstructionScript
  (Effects|Components|ParticleSystem|SetTemplate (Variables|Default|GetValidSteam) (Variables|Default|GetSteamTrickle)))
