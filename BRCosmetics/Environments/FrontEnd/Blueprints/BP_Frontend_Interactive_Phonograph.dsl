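; Interactive frontend Midas phonograph that is cranked up through special interactions, then plays a music track.
; A crank counter selects between the CrankSpin timeline with its sound and, at count two, the PlayRecord timeline.
; PlayRecord spins the record and crank, swings the arm, toggles note particles and fires CompletedPhono from its OpenPeeley track.
; An interaction flag blocks input during animations, and CompletedPeek resets the counter and flag and sends a quest stat event.
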(event EventBeginPlay
  ;[1 1 1] EVENT - BEGINPLAY
  (Variables|Default|SetQuestManager (Quests|GetQuestManager (Fortnite|GetFirstLocalFortPlayerController true) "Invalid")))

(event EventHandleObjectSpecialInteraction
  (params (ActionTag :type Gameplay%20Tag%20Structure) (CameraTag :type Gameplay%20Tag%20Structure))
  (bind _pfx_phonograph_notes (Variables|Default|GetPFX_Phonograph_Notes))
  (if (Variables|Default|GetIsInteractAllowed)
    (Variables|Default|SetIsInteractAllowed false)
    (Utilities|FlowControl|Sequence
      (:then_0
        (switch int (Variables|Default|GetCrankInt)
          (:0
            _)
          (:1
            _)
          (:2
            ;[1 0 0] Music audio
            (Audio|PlaySoundatLocation (Variables|Default|GetMusicTrack) (Transformation|GetActorLocation))
            (bind (_direction _rotate_disc _armanim) (Timeline|PlayRecord PlayFromStart)
              (:Update
                (Transformation|AddRelativeRotation (Variables|Default|GetSM_Golden_Record) (Math|Rotator|MakeRotator 0.0 0.0 (* _rotate_disc 5e+01)) false false)
                (Transformation|SetRelativeRotation (Variables|Default|GetMidas_RecordPlayer_Arm) (Math|Rotator|Lerp(Rotator) (Variables|Default|GetRestRot) (Variables|Default|GetPlayRot) _armanim) false false)
                (Transformation|AddRelativeRotation (Variables|Default|GetMidas_RecordPlayer_Crank) (Math|Rotator|MakeRotator (* _rotate_disc -28.0)) false false))
              (:Finished
                (Components|Activation|Deactivate _pfx_phonograph_notes))
              (:Enable%20Particles
                (Components|Activation|Activate _pfx_phonograph_notes false))
              (:OpenPeeley
                (Default|CallCompletedPhono)))))
        (Math|Integer|IncrementInt (Variables|Default|GetCrankInt)
          (:%20%20
            ;[1 0 0] Cranking Audio
            (Audio|PlaySoundatLocation (Variables|Default|GetCrankSound) (Transformation|GetActorLocation))
            (bind (_direction_2 _newtrack_0) (Timeline|CrankSpin PlayFromStart)
              (:Update
                (Transformation|AddRelativeRotation (Variables|Default|GetMidas_RecordPlayer_Crank) (Math|Rotator|MakeRotator (* _newtrack_0 (* (Variables|Default|GetCrankInt) -2.5))) false false))
              (:Finished
                (Variables|Default|SetIsInteractAllowed true))))))
      (:then_1
        (CallFunction|CommitonSpecialInteraction)))))

(event Custom|CompletedPeek
  (Variables|Default|SetCrankInt 0)
  (Variables|Default|SetIsInteractAllowed true)
  (Quests|SendClientStatEvent (Variables|Default|GetQuestManager)))

(event Custom|Commit_OnSpecialInteraction)

(fn CanCommitNavigationRequest
  (return true))

(fn ConstructionScript)
