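; Interactive frontend cat toy Blueprint that gets pushed a step further on each special interaction, ending in a final tip.
; A DartInt counter picks one of four timelines, each playing a push sound and lerping the mesh between position markers with a rocking pitch.
; The fourth push resets the counter, then on finish fires FinalPushDone and plays a stinger.
; An IsInteractAllowed flag blocks input during pushes, with a delay intended to re-enable it.
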(event EventBeginPlay)

(event EventHandleObjectSpecialInteraction
  (params (ActionTag :type Gameplay%20Tag%20Structure) (CameraTag :type Gameplay%20Tag%20Structure))
  (bind _dartint (Variables|Default|GetDartInt))
  (if (Variables|Default|GetIsInteractAllowed)
    (Variables|Default|SetIsInteractAllowed false)
    (Utilities|FlowControl|Sequence
      (:then_0
        (switch int (Variables|Default|GetDartInt)
          (:0
            (Math|Integer|IncrementInt _dartint
              (:%20%20
                (Audio|PlaySound2D "/BRCosmetics/Sounds/Other/Fort_UI_Sounds/S12_Frontend/UI_Meowscles_Kitty_Push_Cue.UI_Meowscles_Kitty_Push_Cue")
                (bind (_direction _movetime _rocktime) (Timeline|Timeline_0 PlayFromStart)
                  (:Update
                    (Transformation|SetRelativeLocationAndRotation (Variables|Default|GetCISFixMesh) (Math|Vector|Lerp(Vector) "0, 0, 0" (Variables|/Script/Engine.SceneComponent|GetRelativeLocation (Variables|Default|GetPos1)) _movetime) (Math|Rotator|Lerp(Rotator) "0, 0, 0" "20.000000, 0, 0" _rocktime) false false))
                  (:Finished
                    _)))))
          (:1
            (Math|Integer|IncrementInt _dartint
              (:%20%20
                (Audio|PlaySound2D "/BRCosmetics/Sounds/Other/Fort_UI_Sounds/S12_Frontend/UI_Meowscles_Kitty_Push_Cue.UI_Meowscles_Kitty_Push_Cue")
                (bind (_direction_2 _movetime_2 _rocktime_2) (Timeline|Timeline_2 PlayFromStart)
                  (:Update
                    (Transformation|SetRelativeLocationAndRotation (Variables|Default|GetCISFixMesh) (Math|Vector|Lerp(Vector) (Variables|/Script/Engine.SceneComponent|GetRelativeLocation (Variables|Default|GetPos1)) (Variables|/Script/Engine.SceneComponent|GetRelativeLocation (Variables|Default|GetPos2)) _movetime_2) (Math|Rotator|Lerp(Rotator) "0, 0, 0" "15.000000, 0, 0" _rocktime_2) false false))
                  (:Finished
                    _)))))
          (:2
            (Math|Integer|IncrementInt _dartint
              (:%20%20
                (Audio|PlaySound2D "/BRCosmetics/Sounds/Other/Fort_UI_Sounds/S12_Frontend/UI_Meowscles_Kitty_Push_Cue.UI_Meowscles_Kitty_Push_Cue")
                (Components|Timeline|SetPlayRate (Variables|BP_Frontend_IntertiveCatToy|GetPush3) 1.5)
                (bind (_direction_3 _movetime_3 _rocktime_3) (Timeline|Push3 PlayFromStart)
                  (:Update
                    (Transformation|SetRelativeLocationAndRotation (Variables|Default|GetCISFixMesh) (Math|Vector|Lerp(Vector) (Variables|/Script/Engine.SceneComponent|GetRelativeLocation (Variables|Default|GetPos2)) (Variables|/Script/Engine.SceneComponent|GetRelativeLocation (Variables|Default|GetPos3)) _movetime_3) (Math|Rotator|Lerp(Rotator) "0, 0, 0" "12.000000, 0, 0" _rocktime_3) false false))
                  (:Finished
                    _)))))
          (:3
            (Variables|Default|SetDartInt 0)
            (Audio|PlaySound2D "/BRCosmetics/Sounds/Other/Fort_UI_Sounds/S12_Frontend/UI_Meowscles_Kitty_Push_Cue.UI_Meowscles_Kitty_Push_Cue")
            (bind (_direction_4 _movetime_4 _rocktime_4) (Timeline|Timeline_3 PlayFromStart)
              (:Update
                (Transformation|SetRelativeLocationAndRotation (Variables|Default|GetCISFixMesh) (Math|Vector|Lerp(Vector) (Variables|/Script/Engine.SceneComponent|GetRelativeLocation (Variables|Default|GetPos3)) (Variables|/Script/Engine.SceneComponent|GetRelativeLocation (Variables|Default|GetPos4)) _movetime_4) (Math|Rotator|Lerp(Rotator) "0, 0, 0" "150.000000, 0, 0" _rocktime_4) false false))
              (:Finished
                (Default|CallFinalPushDone)
                (Audio|PlaySound2D (Variables|Default|GetStingerSound))))))
        (Utilities|FlowControl|Delay 0.5)
        (Variables|Default|SetIsInteractAllowed true))
      (:then_1
        (CallFunction|CommitonSpecialInteraction)))))

(event Custom|Commit_OnSpecialInteraction)

(fn CanCommitNavigationRequest
  (return true))

(fn ConstructionScript)
