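; Frontend speed bag prop, named Meow, that wobbles and plays a 2D sound on special interaction.
; BeginPlay creates a dynamic punch-bag material on the mesh and stores the initial rotation.
; An allowed interaction locks further input, then a timeline drives the material's Animation scalar, scaled by clamped WobbleStrength, so the wobble is material-driven.
; The timeline's Retrigger event unlocks input, and each hit also calls the commit interaction hook.
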(event EventBeginPlay
  (Variables|Default|SetInitialRotation (Transformation|GetActorRotation))
  (bind _returnvalue (Rendering|Material|CreateDynamicMaterialInstance (Variables|Default|GetCISFixMesh) 0 "/BRCosmetics/Environments/FrontEnd/Materials/BP12/M_PunchBagInteraction_Inst.M_PunchBagInteraction_Inst" "None" :!class /Script/Engine.PrimitiveComponent))
  (Variables|Default|SetBobblinMat _returnvalue))

(event EventHandleObjectSpecialInteraction
  (params (ActionTag :type Gameplay%20Tag%20Structure) (CameraTag :type Gameplay%20Tag%20Structure))
  (if (Variables|Default|GetIsInteractAllowed)
    (Variables|Default|SetIsInteractAllowed false)
    (Utilities|FlowControl|Sequence
      (:then_0
        (Audio|PlaySound2D "/BRCosmetics/Sounds/Other/Fort_UI_Sounds/S12_Frontend/Meow_Speedbag_Cue.Meow_Speedbag_Cue")
        (bind (_direction _wobbletime) (Timeline|Bobblin' PlayFromStart)
          (:Update
            (Rendering|Material|SetScalarParameterValue (Variables|Default|GetBobblinMat) "Animation" (* _wobbletime (Math|Float|Clamp(Float) (Variables|Default|GetWobbleStrength) 1.0 5.0)) :!class /Script/Engine.MaterialInstanceDynamic))
          (:Retrigger
            (Variables|Default|SetIsInteractAllowed true))))
      (:then_1
        (CallFunction|CommitonSpecialInteraction)))))

(event Custom|Commit_OnSpecialInteraction)

(fn CanCommitNavigationRequest
  (return true))

(fn ConstructionScript)
