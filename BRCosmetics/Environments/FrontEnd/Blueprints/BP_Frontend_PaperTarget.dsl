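; Interactive frontend paper target that plays a poster pop and a thrown cleaver animation when a special interaction is triggered.
; Interaction is gated by a flag that is cleared on first use and never reset here.
; A ScalePop timeline tilts the poster, then chains into a CleverFlight timeline that moves and rotates the axe mesh from its cached location to ArrowTarget, with a hit sound.
; The interaction commit call follows the animation start without waiting for completion.
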(event EventBeginPlay
  (Variables|Default|SetInitialRotation (Transformation|GetActorRotation))
  (Variables|Default|SetInitialScale (Transformation|GetActorScale3D))
  (Variables|Default|SetInitialCleverLoc (Variables|/Script/Engine.SceneComponent|GetRelativeLocation (Variables|Default|GetCine_Axe_Gadget_Cine_Henchman_Axe))))

(event EventHandleObjectSpecialInteraction
  (params (ActionTag :type Gameplay%20Tag%20Structure) (CameraTag :type Gameplay%20Tag%20Structure))
  (bind _cine_axe_gadget_cine_henchman_axe (Variables|Default|GetCine_Axe_Gadget_Cine_Henchman_Axe))
  (if (Variables|Default|GetIsInteractAllowed)
    (Variables|Default|SetIsInteractAllowed false)
    (Utilities|FlowControl|Sequence
      (:then_0
        (Audio|PlaySoundatLocation "/BRCosmetics/Sounds/Other/Fort_UI_Sounds/S12_Frontend/UI_InWorld_Paper_Foley_Cue.UI_InWorld_Paper_Foley_Cue" (Transformation|GetActorLocation))
        (bind (_direction _poptime) (Timeline|ScalePop PlayFromStart)
          (:Update
            (Transformation|SetRelativeRotation (Variables|Default|GetS_Poster_01_BP12) (Math|Rotator|Lerp(Rotator) "0, 0, 0" "8.000000,0.000000, 0" _poptime) false false))
          (:Finished
            (Development|PrintString "POP IT!")
            (bind (_direction_2 _flighttime) (Timeline|CleverFlight PlayFromStart)
              (:Update
                (Transformation|SetRelativeLocation _cine_axe_gadget_cine_henchman_axe (Math|Vector|Lerp(Vector) (Variables|Default|GetInitialCleverLoc) (Variables|/Script/Engine.SceneComponent|GetRelativeLocation (Variables|Default|GetArrowTarget)) _flighttime) false false)
                (Transformation|SetRelativeRotation _cine_axe_gadget_cine_henchman_axe (Math|Rotator|Lerp(Rotator) "-60.000000,90.000000, 0" "50.000000,90.000000, 0" _flighttime) false false))
              (:PlaySound
                (Audio|PlaySoundatLocation (Variables|Default|GetCleaverHitSound) (Transformation|GetActorLocation)))))))
      (:then_1
        (CallFunction|CommitonSpecialInteraction)))))

(event Custom|Commit_OnSpecialInteraction)

(fn CanCommitNavigationRequest
  (return true))

(fn ConstructionScript)
