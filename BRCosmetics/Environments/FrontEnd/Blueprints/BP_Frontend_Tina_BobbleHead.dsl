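; Frontend Tina bobblehead prop that plays a sound and wobbles when the player triggers its special interaction.
; A timeline drives the "Animation" scalar on a dynamic material instance, scaled by a clamped click count so repeated clicks wobble harder.
; Interaction locks until the timeline's Retrigger event, and a later ResetWobbleStrength event clears the count.
; Each accepted interaction also calls an empty commit hook, and navigation requests are always allowed.
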(event EventBeginPlay
  (Variables|Default|SetInitialRotation (Transformation|GetActorRotation))
  (bind _returnvalue (Rendering|Material|CreateDynamicMaterialInstance (Variables|Default|GetCISFixMesh) 0 "/BRCosmetics/Environments/FrontEnd/Materials/BP12/M_BobHeadInteraction.M_BobHeadInteraction" "None" :!class /Script/Engine.PrimitiveComponent))
  (Variables|Default|SetBobblinMat _returnvalue))

(event EventHandleObjectSpecialInteraction
  (params (ActionTag :type Gameplay%20Tag%20Structure) (CameraTag :type Gameplay%20Tag%20Structure))
  (if (Variables|Default|GetIsInteractAllowed)
    (Variables|Default|SetIsInteractAllowed false)
    (Utilities|FlowControl|Sequence
      (:then_0
        (Audio|PlaySound2D "/BRCosmetics/Sounds/Other/Quests/SpyBaseSecrets/Tina_Bobble.Tina_Bobble")
        (bind _result (Math|Integer|IncrementInt (Variables|Default|GetCurrentClickNum))
          (:%20%20
            (Variables|Default|SetWobbleStrength (Math|Conversions|ToFloat(Integer) _result))
            (bind (_direction _wobbletime) (Timeline|Bobblin' PlayFromStart)
              (:Update
                (Rendering|Material|SetScalarParameterValue (Variables|Default|GetBobblinMat) "Animation" (* _wobbletime (Math|Float|Clamp(Float) (Variables|Default|GetWobbleStrength) 1.0 5.0)) :!class /Script/Engine.MaterialInstanceDynamic))
              (:Retrigger
                (Variables|Default|SetIsInteractAllowed true))
              (:ResetWobbleStrength
                (Variables|Default|SetCurrentClickNum 0))))))
      (:then_1
        (CallFunction|CommitonSpecialInteraction)))))

(event Custom|Commit_OnSpecialInteraction)

(fn CanCommitNavigationRequest
  (return true))

(fn ConstructionScript)
