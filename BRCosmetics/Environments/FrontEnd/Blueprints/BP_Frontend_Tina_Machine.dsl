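; Interactive front-end prop for Tina's machine that escalates with each click and can reveal a secret fridge.
; Each allowed interaction activates a staged audio cue, raises material shake, turns that stage's dial via a timeline-driven dynamic material parameter, and sets a shutdown timer.
; Shutdown resets clicks and shake, playing either a normal sound or, at max clicks, a blow-up sound and explosion before triggering the fridge secret.
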(event EventBeginPlay
  (Variables|Default|SetInitialRotation (Transformation|GetActorRotation))
  (bind _returnvalue (Rendering|Material|CreateDynamicMaterialInstance (Variables|Default|GetCISFixMesh) 0 "/BRCosmetics/Environments/FrontEnd/Materials/BP12/M_UsableDials.M_UsableDials" "None" :!class /Script/Engine.PrimitiveComponent))
  (Variables|Default|SetBobblinMat _returnvalue))

(event EventHandleObjectSpecialInteraction
  (params (ActionTag :type Gameplay%20Tag%20Structure) (CameraTag :type Gameplay%20Tag%20Structure))
  (if (Variables|Default|GetIsInteractAllowed)
    (Variables|Default|SetIsInteractAllowed false)
    (Utilities|FlowControl|Sequence
      (:then_0
        (Math|Integer|IncrementInt (Variables|Default|GetCurrentClickNum)
          (:%20%20
            (Components|Activation|Activate (Utilities|Select :Index (Variables|Default|GetCurrentClickNum) :Option%200 (Variables|Default|GetTina_Machine_Stage1_Cue) :Option%201 (Variables|Default|GetTina_Machine_Stage2_Cue) :Option%202 (Variables|Default|GetTina_Machine_Stage3_Cue)) false)
            (bind _returnvalue (Utilities|Time|SetTimerbyEvent (delegate Custom|TurnOffMachine) (Variables|Default|GetMachineOnTime)))
            (Variables|Default|SetStayOnTime _returnvalue)
            (Rendering|Material|SetScalarParameterValue (Variables|Default|GetBobblinMat) "Shake" (Utilities|Select :Index (Variables|Default|GetCurrentClickNum) :Option%200 0.5 :Option%201 0.75 :Option%202 1.0) :!class /Script/Engine.MaterialInstanceDynamic)
            (bind (_direction _wobbletime) (Timeline|TurnDial PlayFromStart)
              (:Update
                (Rendering|Material|SetScalarParameterValue (Variables|Default|GetBobblinMat) (Utilities|Select :Index (Variables|Default|GetCurrentClickNum) :Option%200 (Variables|Default|GetDial_1) :Option%201 (Variables|Default|GetDial_2) :Option%202 (Variables|Default|GetDial_3)) (Math|Float|Lerp 0.0 0.4 _wobbletime) :!class /Script/Engine.MaterialInstanceDynamic))
              (:Finished
                (Utilities|FlowControl|Delay 0.5)
                (if (< (Variables|Default|GetCurrentClickNum) (Variables|Default|GetMaxClicks))
                  (Variables|Default|SetIsInteractAllowed true)))))))
      (:then_1
        (CallFunction|CommitonSpecialInteraction)))))

(event Custom|TurnOffMachine
  (bind _maxclicks (Variables|Default|GetMaxClicks))
  (bind _currentclicknum (Variables|Default|GetCurrentClickNum))
  (bind _returnvalue (>= _currentclicknum _maxclicks))
  (Development|PrintString "TurnOff")
  (Components|Activation|Deactivate (Variables|Default|GetTina_Machine_Stage1_Cue))
  (Audio|PlaySoundatLocation (select _returnvalue (Variables|Default|GetBlowUpSound) (Variables|Default|GetNormalTurnOffSound)) (Transformation|GetActorLocation))
  (if _returnvalue
    (Components|Activation|Activate (Variables|Default|GetP_Ashton_Indigo_Explosion) false)
    (Utilities|FlowControl|Sequence
      (:then_0
        (Utilities|FlowControl|Delay 1.0)
        (Class|BPTinaSecretFridge|TriggerFridgeSecret (Variables|Default|GetFridgeDoorOpen)))
      (:then_1
        _)))
  (Variables|Default|SetCurrentClickNum -1)
  (Rendering|Material|SetScalarParameterValue (Variables|Default|GetBobblinMat) "Shake" 0.0 :!class /Script/Engine.MaterialInstanceDynamic))

(event Custom|Commit_OnSpecialInteraction)

(fn CanCommitNavigationRequest
  (return true))

(fn ConstructionScript)
